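# UI material that draws a matchmaking settings slider with a rounded track, fill bar, handle and small circle marker.
# Value and DefaultValue scalar parameters feed named reroutes for the bar, handle and circle masks, sized in pixels through pixel-to-normalized functions.
# Layered lerps blend overlay backplate colors and opacities with a near-black tint.
# Color and opacity are both multiplied by rounded mask and vertex alpha for AlphaComposite blending.
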Material /MatchmakingSettingsUI/MatchmakingSettings/SettingEntries/Slider/Materials/M_UIKit_Slider
graph {
 properties {bAutomaticallySetUsageInEditor=false, blendMode=BLEND_AlphaComposite, materialDomain=MD_UI, usedShadingModels=Default Lit}
 nodes [
  n1 = MaterialFunctionCall() {materialFunction=/MatchmakingSettingsUI/MatchmakingSettings/Materials/MF_UI_MM_OverlayBackplates}
  n2 = LinearInterpolate(A=@n1.B Weak Opacity, B=@n1.B Weak Opacity, Alpha=NamedRerouteUsage())
  n3 = NamedRerouteUsage()
  n4 = Multiply(
   A=VertexColor().A
   B=MaterialFunctionCall(Corner Radius=Constant() {r=10}, Radius in Pixels=Constant() {r=1}) {materialFunction=&MF_UI_CoreUI_RoundedMask}.Shape Mask
  )
  Multiply(A=LinearInterpolate(A=LinearInterpolate(A=@n2, B=@n1.C Opacity, Alpha=@n3) {constB=0}, Alpha=NamedRerouteUsage()), B=@n4) {constA=1}
  n5 = LinearInterpolate(A=@n1.B Weak Color, B=@n1.B Weak Color, Alpha=NamedRerouteUsage())
  n6 = NamedRerouteUsage()
  Multiply(
   A=LinearInterpolate(
    A=LinearInterpolate(A=@n5, B=@n1.C Color, Alpha=@n6) {constB=0}
    B=Constant3Vector() {constant=<rgba 0.0048 0.004 0.0116 1>}
    Alpha=NamedRerouteUsage()
   )
   B=@n4
  )
  NamedRerouteDeclaration(None=ScalarParameter() {defaultValue=0.4, parameterName=Value, sliderMax=1}) {name=Value, nodeColor=<rgba 0 0 0 1>}
  n7 = MaterialFunctionCall() {materialFunction=&GetUserInterfaceUV}
  n8 = NamedRerouteUsage()
  n9 = NamedRerouteUsage()
  n10 = Multiply(A=LinearInterpolate(A=NamedRerouteUsage(), B=NamedRerouteUsage(), Alpha=NamedRerouteUsage()) {constA=256, constB=4}) {constB=2}
  n11 = NamedRerouteUsage()
  n12 = AppendVector(A=Add(A=LinearInterpolate(A=@n8, B=@n9, Alpha=@n10), B=Multiply(A=@n11) {constB=0.5}), B=Constant())
  n13 = Constant2Vector() {g=16, r=3}
  MaterialExpressionMaterialFunctionCall_19 = MaterialFunctionCall!(Value in Pixels=AppendVector(A=@n13, B=@n13), Pixel Reference) {materialFunction=/FortUILibrary/MaterialFunctions/Basics/MF_UI_PixelsToNormalizedV4}
  n14 = Convert(None=@MaterialExpressionMaterialFunctionCall_19."Per‑Axis Normalized (V4)") {
   nodeName=Break Float2
   convertInputs [{
     type=Vector2
     expressionInput {
      expression=&self:MaterialExpressionMaterialFunctionCall_19
      inputName=None
      mask=0
      maskA=0
      maskB=0
      maskG=0
      maskR=0
      outputIndex=1
     }
    }]
   convertOutputs [{type=Scalar}, {type=Scalar}]
  }
  n15 = OneMinus(None=@n14)
  NamedRerouteDeclaration(None=MaterialFunctionCall(
    UVs=Add(A=@n7.Normalized UV, B=@n12)
    Size=AppendVector(A=@n14, B=@n15)
    CornerRoundness=Constant() {r=0.075}
    KeepAspectRatio=StaticBool() {value=true}
   ) {materialFunction=&MF_UI_Box}.Fill) {name=Handle Mask, nodeColor=<rgba 0 0 0 1>}
  NamedRerouteDeclaration(None=@n13.R) {name=Handle Width, nodeColor=<rgba 0 0 0 1>}
  n16 = MaterialFunctionCall() {materialFunction=&GetUserInterfaceUV}
  n17 = Constant() {r=4}
  n18 = ComponentMask(None=MaterialFunctionCall(Value in Pixels=@n17) {materialFunction=&MF_UI_PixelsToNormalizedV4}."Per‑Axis Normalized (V4)") {g=true, r=true}
  n19 = ComponentMask(None=@n18) {r=true}
  n20 = Multiply(A=OneMinus(None=@n19)) {constB=0.5}
  n21 = Multiply(A=@n19) {constB=-0.5}
  n22 = AppendVector(A=Add(A=NamedRerouteUsage()) {constB=8}, B=@n17)
  MaterialExpressionMaterialFunctionCall_49 = MaterialFunctionCall(Value in Pixels=AppendVector(A=@n22, B=@n22)) {materialFunction=&MF_UI_PixelsToNormalizedV4}
  n23 = Convert(None=@MaterialExpressionMaterialFunctionCall_49."Per‑Axis Normalized (V4)") {
   nodeName=Break Float2
   convertInputs [{
     type=Vector2
     expressionInput {
      expression=&self:MaterialExpressionMaterialFunctionCall_49
      inputName=None
      mask=0
      maskA=0
      maskB=0
      maskG=0
      maskR=0
      outputIndex=1
     }
    }]
   convertOutputs [{type=Scalar}, {type=Scalar}]
  }
  n24 = OneMinus(None=@n18) {desc=Max Size}
  n25 = LinearInterpolate(
   A=AppendVector(A=NamedRerouteDeclaration(None=@n23) {name=Min Bar Width, nodeColor=<rgba 0 0 0 1>}, B=OneMinus(None=@n23)) {desc=Min Size}
   B=@n24
   Alpha=NamedRerouteUsage()
  ) {constA=256, constB=4}
  n26 = AppendVector(
   A=LinearInterpolate(
    A=NamedRerouteDeclaration(None=@n20) {name=Min Offset, nodeColor=<rgba 0 0 0 1>}
    B=NamedRerouteDeclaration(None=@n21) {name=Max Offset, nodeColor=<rgba 0 0 0 1>}
    Alpha=ComponentMask(None=@n25) {r=true}
   ) {constA=0.5, constB=-0.005, desc=Offset based on size}
   B=Constant()
  )
  n27 = Constant4Vector() {constant=<rgba 15 15 4 4>}
  n28 = Constant() {r=9}
  n29 = NamedRerouteUsage()
  NamedRerouteDeclaration(None=MaterialFunctionCall(
    UV=Add(A=@n16.Normalized UV, B=@n26)
    Size=@n25
    Corner Radius=LinearInterpolate(A=@n27.RGBA, B=@n28, Alpha=@n29)
    Radius in Pixels=Constant() {r=1}
    Corners Modulated by Size=Constant() {r=0.5}
   ) {materialFunction=&MF_UI_CoreUI_RoundedMask}.Shape Mask) {name=Bar Mask, nodeColor=<rgba 0 0 0 1>}
  NamedRerouteDeclaration(None=ComponentMask(None=@n24) {r=true}) {name=Max Bar Width, nodeColor=<rgba 0 0 0 1>}
  NamedRerouteDeclaration(None=ScalarParameter() {defaultValue=0.25, parameterName=DefaultValue, sliderMax=1}) {name=Default Value, nodeColor=<rgba 0 0 0 1>}
  n30 = NamedRerouteUsage()
  n31 = NamedRerouteUsage()
  n32 = NamedRerouteUsage()
  n33 = LinearInterpolate(
   A=NamedRerouteUsage()
   B=NamedRerouteUsage()
   Alpha=Multiply(A=LinearInterpolate(A=@n30, B=@n31, Alpha=@n32) {constA=256, constB=4}) {constB=2}
  )
  n34 = Constant()
  n35 = MaterialFunctionCall(UV=Add(A=MaterialFunctionCall() {materialFunction=&GetUserInterfaceUV}.Normalized UV, B=AppendVector(A=@n33, B=@n34))) {materialFunction=&MF_UI_UVScaling}
  n36 = NamedRerouteDeclaration(None=MaterialFunctionCall!(Value PX=Constant() {r=8}, UV Pixel Size, X or Y=Constant() {r=1}) {materialFunction=/FortUILibrary/MaterialFunctions/Basics/MF_UI_PixelsToNormalized}."Normalized (>1px)") {name=Circle Size, nodeColor=<rgba 0 0 0 1>}
  n37 = NamedRerouteUsage()
  n38 = Multiply(A=Min(A=@n37, B=OneMinus(None=@n37))) {constB=2}
  NamedRerouteDeclaration(None=Multiply(
    A=MaterialFunctionCall(UVs=@n35.Scaled to Fit, Size=@n36) {materialFunction=&MF_UI_Circle}.Fill
    B=Ceil(None=@n38)
   )) {name=Circle Mask, nodeColor=<rgba 0 0 0 1>}
 ]
}
