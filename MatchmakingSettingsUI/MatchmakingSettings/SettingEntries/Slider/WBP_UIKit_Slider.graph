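# Slider widget for matchmaking settings entries, extending WBP_UIKit_Slider_Base with a material-drawn bar image.
# On Construct it caches a dynamic material instance from Img_Slider, which uses MI_UIKit_Slider.
# Value changes from InteractionLayer_Slider feed that material's "Value" scalar parameter, so the shader renders the bar.
# PreConstruct applies mobile styling by making the image taller when IsUsingCompactUI is true.
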WidgetBlueprint /MatchmakingSettingsUI/MatchmakingSettings/SettingEntries/Slider/WBP_UIKit_Slider
graphs [
 EventGraph {nodes [
   n1 = "On Value Changed (InteractionLayer_Slider)"!() -> (OutputDelegate, then, Value:float) {is=title K2Node_ComponentBoundEvent}
   > UpdateBarMaterial!(execute, self, Value:double=@n1.Value) -> (then)
   Event PreConstruct()
   > SetMobileStyle() {is=title K2Node_CallFunction}
   Event Construct()
   > n2 = GetDynamicMaterial(self=Get Img_Slider() -> (Img_Slider:Image Object Reference))
   > Set BarMaterialInstance(BarMaterialInstance:Material Instance Dynamic Object Reference=@n2) -> (Output_Get:Material Instance Dynamic Object Reference)
  ]}
 SetMobileStyle {nodes [
   SetMobileStyle() {is=title K2Node_FunctionEntry}
   > SetDesiredSizeOverride!(
    execute
    self:Image Object Reference=Get Img_Slider()
    DesiredSize:Vector 2D Structure=Select(Option 0:Vector 2D Structure=(X=256,Y=36), Option 1:Vector 2D Structure=(X=256,Y=44), Index=IsUsingCompactUI()) -> (ReturnValue:Vector 2D Structure)
   ) -> (then)
  ]}
 UpdateBarMaterial {nodes [
   n1 = UpdateBarMaterial!() -> (then, Value:double=~) {is=title K2Node_FunctionEntry}
   > SetScalarParameterValue(self=Get BarMaterialInstance() -> (BarMaterialInstance:Material Instance Dynamic Object Reference), ParameterName=Value, Value=@n1)
  ]}
]
variables [BarMaterialInstance {type=MaterialInstanceDynamic}]
widgets {
 info {
  inheritedWidgetCount=0
  namedSlotCount=1
  parentClass=&/WBP_UIKit_Slider_Base.WBP_UIKit_Slider_Base_C
  rootWidgetClass=null
  widgetCount=1
 }
 widgets [Img_Slider {
   widgetClassPath=&Image
   bIsVariable=true
   widget=Img_Slider
   properties {
    visibility=SelfHitTestInvisible
    brush {
     bIsDynamicallyLoaded=false
     drawAs=Image
     imageType=FullColor
     mirroring=NoMirror
     resourceName=None
     resourceObject=/MatchmakingSettingsUI/MatchmakingSettings/SettingEntries/Slider/MaterialInstances/MI_UIKit_Slider
     tiling=NoTile
     imageSize=<256 36>
     margin=<ltrb 0 0 0 0>
     outlineSettings {
      bUseBrushTransparency=false
      roundingType=HalfHeightRadius
      width=0
      color {colorUseRule=UseColor_Specified, specifiedColor=<rgba 0 0 0 0>}
      cornerRadii=<0 0 0 0>
     }
     tintColor {colorUseRule=UseColor_Specified, specifiedColor=<rgba 1 1 1 1>}
     uVRegion {bIsValid=false, max=<0 0>, min=<0 0>}
    }
   }
  }]
}
